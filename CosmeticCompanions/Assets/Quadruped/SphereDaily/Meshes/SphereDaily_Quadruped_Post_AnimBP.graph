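# Post-process Anim Blueprint for the SphereDaily quadruped cosmetic companion, adding facial pose-driven deformation and secondary physics on the lips and ears.
# The input pose runs through a Pose Blend node using the head facial PoseAsset variable.
# A serial chain of root-relative Anim Dynamics nodes then simulates upper-lip and ear bones on both sides with angular springs, limits and custom gravity.
# The event graph is empty.
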AnimBlueprint /CosmeticCompanions/Assets/Quadruped/SphereDaily/Meshes/SphereDaily_Quadruped_Post_AnimBP
target_skeleton=&/SphereDaily_Quadruped_Skeleton
graphs [
 AnimGraph {nodes [
   n1 = Input Pose()
   n2 = Get PoseAsset() -> (PoseAsset:Pose Asset Object Reference)
   n3 = Anim Dynamics - Bone: dyn_lip_upper_01_l!(
    ComponentPose=Local To Component(LocalPose=Pose!(PoseAsset:Pose Asset Object Reference=@n2, SourcePose:Pose Link Structure=@n1) -> (Pose:Pose Link Structure) {is=title AnimGraphNode_PoseBlendNode, bIgnorePoseAssetSourceAnimationCurves=true, blendOption=Linear})
    Alpha=1
    GravityOverride:Vector="0,0,-75"
   ) -> (Pose) {
    angularDampingOverride=1
    angularSpringConstant=5000
    bAngularSpring=true
    bOverrideAngularDamping=true
    bUseGravityOverride=true
    gravityScale=0.35
    simulationSpace=RootRelative
    alphaBoolBlend {blendOption=Linear}
    boundBone {boneName=dyn_lip_upper_01_l}
    componentAppliedLinearAccClamp=<1000 1000 1000>
    componentLinearAccScale=<1 1 1>
    componentLinearVelScale=<1 1 1>
    gravityOverride {z=-980}
    physicsBodyDefinitions [{
      boundBone {boneName=dyn_lip_upper_01_l}
      boxExtents=<33 33 22>
      constraintSetup {angularTargetAxis=AxisZ, twistAxis=AxisZ, angularLimitsMax=<0 0 0>, angularLimitsMin=<0 -90 0>, angularTarget=<0 0 1>}
      localJointOffset=<0 0 1.5>
     }]
    simSpaceSettings {maxAngularAcceleration=30, maxAngularVelocity=6, simSpaceAngularAlpha=1}
   }
   n4 = Anim Dynamics - Bone: dyn_lip_upper_02_l!(ComponentPose=@n3, Alpha=1, GravityOverride:Vector="0,0,-75") -> (Pose) {
    angularDampingOverride=1
    angularSpringConstant=5000
    bAngularSpring=true
    bOverrideAngularDamping=true
    bUseGravityOverride=true
    gravityScale=0.35
    simulationSpace=RootRelative
    alphaBoolBlend {blendOption=Linear}
    boundBone {boneName=dyn_lip_upper_02_l}
    componentAppliedLinearAccClamp=<1000 1000 1000>
    componentLinearAccScale=<1 1 1>
    componentLinearVelScale=<1 1 1>
    gravityOverride {z=-980}
    physicsBodyDefinitions [{
      boundBone {boneName=dyn_lip_upper_02_l}
      boxExtents=<33 33 22>
      constraintSetup {angularTargetAxis=AxisZ, twistAxis=AxisZ, angularLimitsMax=<0 0 0>, angularLimitsMin=<0 -90 0>, angularTarget=<0 0 1>}
      localJointOffset=<0 0 1.5>
     }]
    simSpaceSettings {maxAngularAcceleration=30, maxAngularVelocity=6, simSpaceAngularAlpha=1}
   }
   n5 = Anim Dynamics - Bone: dyn_lip_upper_03_l!(ComponentPose=@n4, Alpha=1, GravityOverride:Vector="0,0,-75") -> (Pose) {
    angularDampingOverride=1
    angularSpringConstant=5000
    bAngularSpring=true
    bOverrideAngularDamping=true
    bUseGravityOverride=true
    gravityScale=0.35
    simulationSpace=RootRelative
    alphaBoolBlend {blendOption=Linear}
    boundBone {boneName=dyn_lip_upper_03_l}
    componentAppliedLinearAccClamp=<1000 1000 1000>
    componentLinearAccScale=<1 1 1>
    componentLinearVelScale=<1 1 1>
    gravityOverride {z=-980}
    physicsBodyDefinitions [{
      boundBone {boneName=dyn_lip_upper_03_l}
      boxExtents=<33 33 22>
      constraintSetup {angularTargetAxis=AxisZ, twistAxis=AxisZ, angularLimitsMax=<0 0 0>, angularLimitsMin=<0 -90 0>, angularTarget=<0 0 1>}
      localJointOffset=<0 0 1.5>
     }]
    simSpaceSettings {maxAngularAcceleration=30, maxAngularVelocity=6, simSpaceAngularAlpha=1}
   }
   n6 = Anim Dynamics - Bone: dyn_lip_upper_01_r!(ComponentPose=@n5, Alpha=1, GravityOverride:Vector="0,0,-75") -> (Pose) {
    angularDampingOverride=1
    angularSpringConstant=5000
    bAngularSpring=true
    bOverrideAngularDamping=true
    bUseGravityOverride=true
    gravityScale=0.35
    simulationSpace=RootRelative
    alphaBoolBlend {blendOption=Linear}
    boundBone {boneName=dyn_lip_upper_01_r}
    componentAppliedLinearAccClamp=<1000 1000 1000>
    componentLinearAccScale=<1 1 1>
    componentLinearVelScale=<1 1 1>
    gravityOverride {z=-980}
    physicsBodyDefinitions [{
      boundBone {boneName=dyn_lip_upper_01_r}
      boxExtents=<33 33 22>
      constraintSetup {angularTargetAxis=AxisZ, twistAxis=AxisZ, angularLimitsMax=<0 0 0>, angularLimitsMin=<0 -90 0>, angularTarget=<0 0 1>}
      localJointOffset=<0 0 -1.5>
     }]
    simSpaceSettings {maxAngularAcceleration=33, maxAngularVelocity=6, simSpaceAngularAlpha=1}
   }
   n7 = Anim Dynamics - Bone: dyn_lip_upper_02_r!(ComponentPose=@n6, Alpha=1, GravityOverride:Vector="0,0,-75") -> (Pose) {
    angularDampingOverride=1
    angularSpringConstant=5000
    bAngularSpring=true
    bOverrideAngularDamping=true
    bUseGravityOverride=true
    gravityScale=0.35
    simulationSpace=RootRelative
    alphaBoolBlend {blendOption=Linear}
    boundBone {boneName=dyn_lip_upper_02_r}
    componentAppliedLinearAccClamp=<1000 1000 1000>
    componentLinearAccScale=<1 1 1>
    componentLinearVelScale=<1 1 1>
    gravityOverride {z=-980}
    physicsBodyDefinitions [{
      boundBone {boneName=dyn_lip_upper_02_r}
      boxExtents=<33 33 22>
      constraintSetup {angularTargetAxis=AxisZ, twistAxis=AxisZ, angularLimitsMax=<0 0 0>, angularLimitsMin=<0 -90 0>, angularTarget=<0 0 1>}
      localJointOffset=<0 0 -1.5>
     }]
    simSpaceSettings {maxAngularAcceleration=33, maxAngularVelocity=6, simSpaceAngularAlpha=1}
   }
   n8 = Anim Dynamics - Bone: dyn_lip_upper_03_r!(ComponentPose=@n7, Alpha=1, GravityOverride:Vector="0,0,-75") -> (Pose) {
    angularDampingOverride=1
    angularSpringConstant=5000
    bAngularSpring=true
    bOverrideAngularDamping=true
    bUseGravityOverride=true
    gravityScale=0.35
    simulationSpace=RootRelative
    alphaBoolBlend {blendOption=Linear}
    boundBone {boneName=dyn_lip_upper_03_r}
    componentAppliedLinearAccClamp=<1000 1000 1000>
    componentLinearAccScale=<1 1 1>
    componentLinearVelScale=<1 1 1>
    gravityOverride {z=-980}
    physicsBodyDefinitions [{
      boundBone {boneName=dyn_lip_upper_03_r}
      boxExtents=<33 33 22>
      constraintSetup {angularTargetAxis=AxisZ, twistAxis=AxisZ, angularLimitsMax=<0 0 0>, angularLimitsMin=<0 -90 0>, angularTarget=<0 0 1>}
      localJointOffset=<0 0 -1.5>
     }]
    simSpaceSettings {maxAngularAcceleration=33, maxAngularVelocity=6, simSpaceAngularAlpha=1}
   }
   n9 = Anim Dynamics - Bone: ear_02_r!(ComponentPose=@n8, Alpha=1, GravityOverride:Vector="0,0,-75") -> (Pose) {
    angularDampingOverride=1
    angularSpringConstant=6500
    bAngularSpring=true
    bOverrideAngularDamping=true
    bUseGravityOverride=true
    gravityScale=0.35
    simulationSpace=RootRelative
    alphaBoolBlend {blendOption=Linear}
    alphaScaleBiasClamp {outRange {max=0.5}}
    boundBone {boneName=ear_02_r}
    componentAppliedLinearAccClamp=<500 500 500>
    componentLinearAccScale=<1 1 1>
    componentLinearVelScale=<1 1 1>
    gravityOverride {z=-980}
    physicsBodyDefinitions [{
      boundBone {boneName=ear_02_r}
      boxExtents=<15 20 20>
      constraintSetup {angularLimitsMax=<0 20 35>, angularLimitsMin=<0 -20 -70>, angularTarget=<1 0 0>}
      localJointOffset=<-0.5 0 0>
     }]
    simSpaceSettings {maxAngularAcceleration=12, maxAngularVelocity=3, simSpaceAngularAlpha=1}
   }
   n10 = Anim Dynamics - Bone: ear_03_r!(ComponentPose=@n9, Alpha=1, GravityOverride:Vector="0,0,-150") -> (Pose) {
    angularDampingOverride=1
    angularSpringConstant=4000
    bAngularSpring=true
    bOverrideAngularDamping=true
    bUseGravityOverride=true
    gravityScale=0.35
    simulationSpace=RootRelative
    alphaBoolBlend {blendOption=Linear}
    boundBone {boneName=ear_03_r}
    componentAppliedLinearAccClamp=<500 500 500>
    componentLinearAccScale=<1 1 1>
    componentLinearVelScale=<1 1 1>
    gravityOverride {z=-980}
    physicsBodyDefinitions [{
      boundBone {boneName=ear_03_r}
      boxExtents=<12 15 15>
      constraintSetup {angularLimitsMax=<0 20 30>, angularLimitsMin=<0 -20 -60>, angularTarget=<1 0 0>}
      localJointOffset=<-1 0 0>
     }]
    simSpaceSettings {maxAngularAcceleration=20, maxAngularVelocity=4, simSpaceAngularAlpha=1}
   }
   n11 = Anim Dynamics - Bone: ear_04_r!(ComponentPose=@n10, Alpha=1, GravityOverride:Vector="0,0,-150") -> (Pose) {
    angularDampingOverride=1
    angularSpringConstant=4000
    bAngularSpring=true
    bOverrideAngularDamping=true
    bUseGravityOverride=true
    gravityScale=0.35
    simulationSpace=RootRelative
    alphaBoolBlend {blendOption=Linear}
    boundBone {boneName=ear_04_r}
    componentAppliedLinearAccClamp=<500 500 500>
    componentLinearAccScale=<1 1 1>
    componentLinearVelScale=<1 1 1>
    gravityOverride {z=-980}
    physicsBodyDefinitions [{
      boundBone {boneName=ear_04_r}
      boxExtents=<10 13 13>
      constraintSetup {angularLimitsMax=<0 20 30>, angularLimitsMin=<0 -20 -60>, angularTarget=<1 0 0>}
      localJointOffset=<-1 0 0.5>
     }]
    simSpaceSettings {maxAngularAcceleration=20, maxAngularVelocity=4, simSpaceAngularAlpha=1}
   }
   n12 = Anim Dynamics - Bone: ear_02_l!(ComponentPose=@n11, Alpha=1, GravityOverride:Vector="0,0,-75") -> (Pose) {
    angularDampingOverride=1
    angularSpringConstant=6500
    bAngularSpring=true
    bOverrideAngularDamping=true
    bUseGravityOverride=true
    gravityScale=0.35
    simulationSpace=RootRelative
    alphaBoolBlend {blendOption=Linear}
    alphaScaleBiasClamp {outRange {max=0.5}}
    boundBone {boneName=ear_02_l}
    componentAppliedLinearAccClamp=<500 500 500>
    componentLinearAccScale=<1 1 1>
    componentLinearVelScale=<1 1 1>
    gravityOverride {z=-980}
    physicsBodyDefinitions [{
      boundBone {boneName=ear_02_l}
      boxExtents=<15 20 20>
      constraintSetup {angularLimitsMax=<0 20 35>, angularLimitsMin=<0 -20 -70>, angularTarget=<1 0 0>}
      localJointOffset=<0.5 0 0>
     }]
    simSpaceSettings {maxAngularAcceleration=12, maxAngularVelocity=3, simSpaceAngularAlpha=1}
   }
   n13 = Anim Dynamics - Bone: ear_03_l!(ComponentPose=@n12, Alpha=1, GravityOverride:Vector="0,0,-150") -> (Pose) {
    angularDampingOverride=1
    angularSpringConstant=4000
    bAngularSpring=true
    bOverrideAngularDamping=true
    bUseGravityOverride=true
    gravityScale=0.35
    simulationSpace=RootRelative
    alphaBoolBlend {blendOption=Linear}
    boundBone {boneName=ear_03_l}
    componentAppliedLinearAccClamp=<500 500 500>
    componentLinearAccScale=<1 1 1>
    componentLinearVelScale=<1 1 1>
    gravityOverride {z=-980}
    physicsBodyDefinitions [{
      boundBone {boneName=ear_03_l}
      boxExtents=<12 15 15>
      constraintSetup {angularLimitsMax=<0 20 30>, angularLimitsMin=<0 -20 -60>, angularTarget=<1 0 0>}
      localJointOffset=<1 0 0>
     }]
    simSpaceSettings {maxAngularAcceleration=20, maxAngularVelocity=4, simSpaceAngularAlpha=1}
   }
   n14 = Anim Dynamics - Bone: ear_04_l!(ComponentPose=@n13, Alpha=1, GravityOverride:Vector="0,0,-150") -> (Pose) {
    angularDampingOverride=1
    angularSpringConstant=4000
    bAngularSpring=true
    bOverrideAngularDamping=true
    bUseGravityOverride=true
    gravityScale=0.35
    simulationSpace=RootRelative
    alphaBoolBlend {blendOption=Linear}
    boundBone {boneName=ear_04_l}
    componentAppliedLinearAccClamp=<500 500 500>
    componentLinearAccScale=<1 1 1>
    componentLinearVelScale=<1 1 1>
    gravityOverride {z=-980}
    physicsBodyDefinitions [{
      boundBone {boneName=ear_04_l}
      boxExtents=<10 13 13>
      constraintSetup {angularLimitsMax=<0 20 30>, angularLimitsMin=<0 -20 -60>, angularTarget=<1 0 0>}
      localJointOffset=<1 0 -0.5>
     }]
    simSpaceSettings {maxAngularAcceleration=20, maxAngularVelocity=4, simSpaceAngularAlpha=1}
   }
   Output Pose(Result=Component To Local(ComponentPose=@n14))
  ]}
 EventGraph {links [], nodes []}
]
variables [PoseAsset {type=PoseAsset, default=&/SphereDaily_Quadruped_Head_Facial_Poses_PoseAsset}]
